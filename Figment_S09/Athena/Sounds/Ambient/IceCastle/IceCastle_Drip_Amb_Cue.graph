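# Sound cue that plays a looping ambient drip sound for the Ice Castle in Figment Season 9 Athena.
# A single Wave Player node loops the IceCastle_Drips_01 wave directly as the root, with no randomization or modulation.
# The cue boosts its volume, uses the dedicated IceCastle_Drip_Att attenuation asset for spatial falloff, and routes through the outdoor ambient sound class.
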SoundCue /Figment_S09/Athena/Sounds/Ambient/IceCastle/IceCastle_Drip_Amb_Cue
root=WavePlayer_0
nodes [{id=WavePlayer_0, properties {bLooping=true, soundWaveAssetPtr=&/IceCastle_Drips_01}}]
properties {
 attenuationSettings=&/IceCastle_Drip_Att
 maxDistance=2304
 soundClassObject=&Fort_Ambient_Outdoors_Athena
 volumeMultiplier=2
}
